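# Additive unlit Niagara sprite material for the Asteria outdoor firepit's heated effect, building a noisy glow pattern from the world position behind translucency.
# The scene's XY drives a packed fire noise texture, which distorts two opposing panners timed by the dynamic Speed parameter.
# A sphere mask around the particle with noise-perturbed Radius, times a DepthDistance height fade, masks the summed panned noise.
# Particle color enters an emissive-adjustment function; several texture and math nodes sit unused.
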Material /Game/Environments/Asteria/Props/Outdoor/Outdoor_Firepit_A/FX/M_FX_HeatedFirepit_Plane
graph {
 properties {bUsedWithNiagaraSprites=true, blendMode=BLEND_Additive, pixelDepthOffsetMode=PDOM_Legacy, shadingModel=MSM_Unlit, usedShadingModels=Unlit}
 nodes [
  ActorPositionWS()
  Add()
  MaterialFunctionCall(Emissive=ParticleColor().RGB) {materialFunction=&MF_FX_EmissiveAdjustments}
  n1 = MaterialFunctionCall() {materialFunction=&WorldPositionBehindTranslucency}
  n2 = ComponentMask(None=@n1.Result) {g=true, r=true}
  n3 = ParticlePositionWS()
  n4 = ComponentMask(None=@n3) {g=true, r=true}
  n5 = DynamicParameter() {paramNames [FireDetailScale, Speed, Radius, Param4]}
  n6 = Divide(A=@n2, B=@n5.FireDetailScale)
  n7 = TextureSample(UVs=@n6) {texture=/Game/Effects/Fort_Effects/Textures/Fire/HF_FireNoisesPacked}
  n8 = Multiply(A=@n7.R) {constB=2}
  n9 = Add(A=@n5.Radius, B=Multiply(A=Subtract(A=@n8)) {constB=2})
  n10 = ComponentMask(None=@n3) {b=true}
  n11 = Add(A=@n10, B=ScalarParameter() {defaultValue=10, parameterName=DepthDistance})
  n12 = ComponentMask(None=@n1.Result) {b=true}
  n13 = Add(A=Multiply(A=@n7.R) {constB=0.1}, B=@n6)
  n14 = Multiply(A=@n5.Speed, B=Time())
  n15 = TextureSample(UVs=Panner(Coordinate=@n13, Time=@n14) {bFractionalPart=true, speedX=0.5, speedY=-0.2}) {texture=&HF_FireNoisesPacked}
  n16 = Panner(Coordinate=@n13, Time=@n14) {bFractionalPart=true, speedX=-0.2, speedY=0.5}
  Multiply(
   A=Multiply(A=SphereMask(A=@n2, B=@n4, Radius=@n9) {attenuationRadius=0.5, hardnessPercent=0}, B=SmoothStep(Min=@n11, Max=@n10, Value=@n12))
   B=Add(A=@n15.G, B=TextureSample(UVs=@n16) {texture=&HF_FireNoisesPacked}.R)
  )
  Multiply(A=@n5.Radius, B=@n9)
  Multiply(B=@n15.G)
  TextureSample() {texture=/Game/Effects/Fort_Effects/Textures/Fire/2DGasTorch_Flipbk_packed}
  TextureSample() {texture=/Game/Effects/Fort_Effects/Textures/Fire/T_FireTile}
 ]
}
